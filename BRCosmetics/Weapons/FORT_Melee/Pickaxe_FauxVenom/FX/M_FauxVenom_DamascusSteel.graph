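# Translucent unlit Niagara ribbon material for FauxVenom pickaxe trail effects, colored by particle color through an emissive-adjustment function.
# The Damascus Pattern switch selects sine stripes, warped by panning smoke noise with particle-driven thickness, or a panning shape texture scaled along the trail with particle-driven X offset.
# Both branches use trail-end and cross-width fades, stripes add an inverted sphere mask, and particle alpha multiplies the result before a SmoothStep.
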Material /BRCosmetics/Weapons/FORT_Melee/Pickaxe_FauxVenom/FX/M_FauxVenom_DamascusSteel
graph {
 properties {bUsedWithNiagaraRibbons=true, blendMode=BLEND_Translucent, pixelDepthOffsetMode=PDOM_Legacy, shadingModel=MSM_Unlit, usedShadingModels=Unlit}
 nodes [
  n1 = OneMinus(None=ComponentMask(None=TextureCoordinate()) {r=true})
  n2 = TextureCoordinate() {coordinateIndex=1}
  n3 = AppendVector(
   A=ScalarParameter() {defaultValue=1.5, parameterName=Distortion Texture Scale X, sliderMax=5}
   B=ScalarParameter() {defaultValue=1.5, parameterName=Distortion Texture Scale Y, sliderMax=5}
  )
  n4 = Panner(
   Coordinate=Multiply(A=Power(Base=@n1) {constExponent=0.5}, B=Multiply(A=@n2, B=@n3))
   Speed=AppendVector(
    A=ScalarParameter() {defaultValue=0.1, parameterName=Speed X, sliderMax=5, sliderMin=-5}
    B=ScalarParameter() {parameterName=Speed Y, sliderMax=5, sliderMin=-5}
   )
  ) {speedX=1}
  n5 = Add(
   A=TextureCoordinate() {coordinateIndex=1}
   B=Multiply(
    A=TextureSampleParameter2D(UVs=@n4) {parameterName=Distortion Texture, texture=&T_Inky_Smoke_Tile}.R
    B=ScalarParameter() {defaultValue=1, parameterName=Distortion Intensity, sliderMax=5}
   )
  )
  n6 = Sine(None=Multiply(A=ComponentMask(None=@n5) {r=true}, B=ScalarParameter() {defaultValue=5, parameterName=Stripe Density, sliderMax=20}) {constB=10})
  ConstantBiasScale(None=@n6)
  DynamicParameter() {defaultValue=<rgba 0 0.5 1 1>, paramNames [Strip Thickness, Offest X, Dissolve Control, Param4]}
  n7 = ParticleColor()
  MaterialFunctionCall(Emissive=@n7.RGB) {materialFunction=&MF_FX_EmissiveAdjustments}
  Multiply() {constA=1}
  ScalarParameter() {parameterName=Strip Thickness}
  ScalarParameter() {defaultValue=1, parameterName=Shape Texture Scale Y, sliderMax=5}
  n8 = OneMinus(None=ComponentMask(None=TextureCoordinate()) {r=true})
  n9 = ScalarParameter() {defaultValue=2, parameterName=Trail End Fade, sliderMax=5}
  n10 = ComponentMask(None=TextureCoordinate()) {r=true}
  n11 = Sine(None=ComponentMask(None=TextureCoordinate()) {g=true}) {period=2}
  n12 = Multiply(A=Multiply(A=Power(Base=@n8, Exp=@n9)), B=Multiply(A=SmoothStep(Value=@n10) {constMax=0.05}, B=Power(Base=@n11) {constExponent=0.5}))
  n13 = TextureCoordinate()
  n14 = Constant2Vector() {g=1, r=0.8}
  n15 = OneMinus(None=SphereMask(A=Multiply(A=@n13, B=@n14), B=Constant2Vector() {g=0.95, r=1.2}) {attenuationRadius=1.1, hardnessPercent=75})
  n16 = DynamicParameter() {defaultValue=<rgba 0 0.5 1 1>, paramNames [Strip Thickness, Offest X, Dissolve Control, Param4]}
  n17 = Multiply(A=Multiply(A=@n12, B=Power(Base=@n15) {constExponent=5}), B=Saturate(None=Add(A=@n6, B=@n16.Strip Thickness)))
  n18 = OneMinus(None=ComponentMask(None=TextureCoordinate()) {r=true})
  n19 = MaterialFunctionCall(UVs=TextureCoordinate() {coordinateIndex=1}, Texture Scale=AppendVector(A=Constant() {r=1}, B=Multiply(A=@n18) {constB=2})) {materialFunction=&ScaleUVsByCenter}
  n20 = AppendVector(A=ComponentMask(None=@n19.UVs) {r=true}, B=Saturate(None=ComponentMask(None=@n19.UVs) {g=true}))
  n21 = AppendVector(A=ScalarParameter() {defaultValue=1, parameterName=Shape Texture Scale X, sliderMax=5}, B=Constant() {r=1})
  n22 = DynamicParameter() {defaultValue=<rgba 0 0.5 1 1>, paramNames [Strip Thickness, Offest X, Dissolve Control, Param4]}
  n23 = Constant()
  n24 = Panner(
   Coordinate=Add(A=Multiply(A=@n20, B=@n21), B=AppendVector(A=@n22.Offest X, B=@n23))
   Speed=AppendVector(A=ScalarParameter() {defaultValue=1, parameterName=Shape Texture Speed X, sliderMax=5}, B=Constant())
  ) {speedX=1}
  n25 = TextureCoordinate()
  n26 = SmoothStep(Value=OneMinus(None=ComponentMask(None=@n25) {r=true})) {constMax=0.9, constMin=0.5}
  n27 = Multiply(
   A=LinearInterpolate(
    A=TextureSample(UVs=@n24) {constCoordinate=1, texture=&/T_Pickaxe_FauxVenom}.R
    B=@n26
    Alpha=@n26
   )
   B=@n12
  )
  SmoothStep(Value=Multiply(A=@n7.A, B=StaticSwitchParameter(True=@n17, False=@n27) {parameterName=Damascus Pattern}))
 ]
}
