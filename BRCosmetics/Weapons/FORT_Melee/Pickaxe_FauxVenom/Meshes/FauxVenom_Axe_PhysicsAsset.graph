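# Physics asset for the FauxVenom two-handed pickaxe cosmetic, a melee weapon skeletal mesh.
# It holds a single capsule body on the Melee_TwoHanded_Handle bone, slightly offset and rotated from the bone origin.
# With no constraints and no other bodies, the weapon has one simple collision volume and no articulated parts.
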PhysicsAsset /BRCosmetics/Weapons/FORT_Melee/Pickaxe_FauxVenom/Meshes/FauxVenom_Axe_PhysicsAsset
bodies {Melee_TwoHanded_Handle [{shapeType=Capsule, length=35.539, radius=8.4026, center=<0.224 2.7772 12.004>, rotation=<rot 3.1398 -90 0>}]}
constraints []
